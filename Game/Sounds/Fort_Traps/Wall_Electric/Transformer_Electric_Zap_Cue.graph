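# Sound cue for an electric zap, filed under the electric wall trap sounds.
# An equally weighted random node picks one of six shared enemy electric medium-impact waves and feeds it through a modulator at the root.
# It uses the traps sound class and attenuation, with a shared concurrency rule named for a three-voice, stop-oldest limit with volume scaling.
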SoundCue /Game/Sounds/Fort_Traps/Wall_Electric/Transformer_Electric_Zap_Cue
root=Modulator_9
nodes [
 {id=Modulator_9, children [Random_0]}
 {id=Random_0, children [WavePlayer_0,WavePlayer_1,WavePlayer_2,WavePlayer_3,WavePlayer_4,WavePlayer_5], properties {weights [1,1,1,1,1,1]}}
 {id=WavePlayer_0, properties {soundWaveAssetPtr=/Game/Sounds/Fort_GamePlay_Sounds/Elemental_Layers/enemy_electric_medium_impact_06}}
 {id=WavePlayer_1, properties {soundWaveAssetPtr=/Game/Sounds/Fort_GamePlay_Sounds/Elemental_Layers/enemy_electric_medium_impact_01}}
 {id=WavePlayer_2, properties {soundWaveAssetPtr=/Game/Sounds/Fort_GamePlay_Sounds/Elemental_Layers/enemy_electric_medium_impact_02}}
 {id=WavePlayer_3, properties {soundWaveAssetPtr=/Game/Sounds/Fort_GamePlay_Sounds/Elemental_Layers/enemy_electric_medium_impact_03}}
 {id=WavePlayer_4, properties {soundWaveAssetPtr=/Game/Sounds/Fort_GamePlay_Sounds/Elemental_Layers/enemy_electric_medium_impact_04}}
 {id=WavePlayer_5, properties {soundWaveAssetPtr=/Game/Sounds/Fort_GamePlay_Sounds/Elemental_Layers/enemy_electric_medium_impact_05}}
]
properties {
 attenuationSettings=&Fort_Traps_Att
 maxDistance=6000
 soundClassObject=&Fort_Traps
 volumeMultiplier=1.1
 attenuationOverrides {falloffDistance=3200, attenuationShapeExtents {x=800}}
 concurrencyOverrides {resolutionRule=PreventNew}
 concurrencySet [/Game/Sounds/Concurrency_Rules/ThreeMaximum_StopOldest_VolumeScale]
}
