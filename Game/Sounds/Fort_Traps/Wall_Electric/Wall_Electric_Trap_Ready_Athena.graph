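# Sound cue for the electric wall trap's ready sound, in the Athena variant.
# A random node picks between two equally weighted ready waves, then a modulator varies pitch on each play.
# It plays in the trap sound class with trap-activation attenuation, and a concurrency override prevents new instances once the limit is reached.
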SoundCue /Game/Sounds/Fort_Traps/Wall_Electric/Wall_Electric_Trap_Ready_Athena
root=Modulator_10
nodes [
 {id=Modulator_10, children [Random_20], properties {pitchMax=1.1}}
 {id=Random_20, children [WavePlayer_71, WavePlayer_72], properties {weights [1,1]}}
 {id=WavePlayer_71, properties {soundWaveAssetPtr=&/WallTrap_Electric_Ready_001}}
 {id=WavePlayer_72, properties {soundWaveAssetPtr=&/WallTrap_Electric_Ready_002}}
]
properties {
 attenuationSettings=/Game/Sounds/Attenuation/Fort_Traps_Activate_Athena_Att
 bOverrideConcurrency=true
 maxDistance=3500
 soundClassObject=&Fort_Traps
 volumeMultiplier=0.5
 attenuationOverrides {falloffDistance=2100}
 concurrencyOverrides {resolutionRule=PreventNew}
}
